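; Behavior tree evaluator for the Masamune AI's Fire Petal Typhoon Blade, deciding whether it may begin.
; On enter, it casts the owning controller to the Athena AI bot controller and caches it with its controlled Phoebe bot pawn.
; Its starting-condition check calls the parent, then returns only the boolean value of the LesserDemonCanUseBlade scalable float.
; The parent result is not used, so that tunable alone gates the behavior.
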(event EventOnEnter
  (params (OwnerComp :type Behavior%20Tree%20Component%20Object%20Reference))
  (bind _ascontroller (Utilities|Casting|CastToController (Components|GetOwner OwnerComp)))
  (bind _asfort_athena_aibot_controller (Utilities|Casting|CastToFortAthenaAIBotController _ascontroller))
  (bind _output_get (Variables|Default|SetAIController _asfort_athena_aibot_controller))
  (bind _asbp_player_pawn_athena_phoebe (Utilities|Casting|CastToBP_PlayerPawn_Athena_Phoebe (Pawn|GetControlledPawn _output_get)))
  (Variables|Default|SetAIPawn _asbp_player_pawn_athena_phoebe))

(fn EvaluateStartingConditions
  (params (OwnerComp :type Behavior%20Tree%20Component%20Object%20Reference))
  (|CallParentFunction OwnerComp)
  (return (ScalableFloat|GetValueasBool (Variables|Default|GetLesserDemonCanUseBlade_Hf))))
